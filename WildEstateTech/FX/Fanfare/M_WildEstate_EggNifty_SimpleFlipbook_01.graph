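# Masked, unlit, two-sided Niagara sprite and mesh-particle material for the EggNifty fanfare effect, using a flipbook texture.
# Texture RGB is multiplied by particle color, feeding both an emissive-adjustment function and a day/night bloom control followed by inverse eye adaptation.
# Texture alpha is scaled by Alpha_Mult and particle alpha.
# The graph has no UV or frame logic of its own.
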Material /WildEstateTech/FX/Fanfare/M_WildEstate_EggNifty_SimpleFlipbook_01
graph {
 properties {
  bAutomaticallySetUsageInEditor=false
  bUsedWithNiagaraMeshParticles=true
  bUsedWithNiagaraSprites=true
  blendMode=BLEND_Masked
  shadingModel=MSM_Unlit
  twoSided=true
  usedShadingModels=Unlit
 }
 nodes [
  Constant()
  n1 = TextureSampleParameter2D() {parameterName=Texture_Flipbook, texture=&/T_FX_EggNifty_Proto_01}
  n2 = ParticleColor()
  n3 = Multiply(A=@n1.RGB, B=@n2.RGB)
  EyeAdaptationInverse(LightValueInput=MaterialFunctionCall!(DayBloom_Mult=Constant() {r=1}, In=@n3) {materialFunction=/Game/Packages/Fortnite_UniqueMaterials_1/MH_Materials/DayNight_Auto_Bloom_Control}.Result)
  MaterialFunctionCall(Emissive=@n3) {materialFunction=&MF_FX_EmissiveAdjustments}
  n4 = ScalarParameter() {defaultValue=1, parameterName=Alpha_Mult}
  Reroute(None=Multiply(A=Multiply(A=@n1.A, B=@n4), B=@n2.A))
 ]
}
