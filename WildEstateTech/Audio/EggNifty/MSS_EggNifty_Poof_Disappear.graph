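# MetaSound source that plays the EggNifty "poof disappear" sound effect as a mono one-shot.
# On Play triggers a single Wave Player, which feeds the mono output and fires On Finished when the wave ends.
# Playback uses the Fortnite chest attenuation settings and the loot-collect sound class, with no randomization or other processing in the graph.
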MetaSoundSource /WildEstateTech/Audio/EggNifty/MSS_EggNifty_Poof_Disappear
inputs [UE.Source.OnPlay {type=Trigger}]
interfaces [UE.OutputFormat.Mono:1.0, UE.Source.OneShot:1.0, UE.Source:1.1]
outputs [
 "UE.OutputFormat.Mono.Audio:0" {type=Audio}
 UE.Source.OneShot.OnFinished {type=Trigger}
]
pages [Default {
  input_defaults {UE.Source.OnPlay=null}
  nodes [
   InitVariable.WaveAsset:Array()
   WavePlayer_0 = UE.Wave Player:Mono(Play=^UE.Source.OnPlay) {also {defaults {Wave Asset=&/EggNifty_Poof_Disappear}, version [1,1]}}
  ]
  links [{dst="@UE.Source.OneShot.OnFinished", src=WavePlayer_0.On Finished}, {dst="@UE.OutputFormat.Mono.Audio:0", src=WavePlayer_0.Out Mono}]
 }]
properties {
 assetImportData=&self:AssetImportData
 attenuationSettings=/Game/Sounds/Attenuation/Fort_Chests_Att
 maxDistance=1300
 sampleRate=48000
 soundClassObject=&Fort_Loot_Collect
}
